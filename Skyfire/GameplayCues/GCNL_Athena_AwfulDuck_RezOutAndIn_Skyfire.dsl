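; GCNL_Athena_AwfulDuck_RezOutAndIn_Skyfire is a looping gameplay cue notify, by its name for a rez-out-and-in effect, with almost no graph logic of its own.
; BeginPlay, overlap, Tick and the construction script are empty.
; The CE_SetPawnHidden event only calls its parent; a comment marks the lens spawn on hiding the pawn as not in use.
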(event EventBeginPlay)

(event Collision|EventActorBeginOverlap
  (params (OtherActor :type Actor%20Object%20Reference)))

(event EventTick
  (params (DeltaSeconds :type Float%20%28single-precision%29)))

(event EventCE_SetPawnHidden
  ;[0 0 0] (NOT IN USE) Spawn the Lens when we set the pawn hidden.
  (|CallParentFunction))

(fn ConstructionScript)
