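; Looping gameplay cue for the start of the Skyfire prison sequence that plays alien VO for the affected player.
; On application it runs the parent logic, then plays the first-person prison warp sound only if the target is a valid, locally controlled Fort player pawn.
; The sound location takes X from the pawn mesh, replaces Y with a configured value and adds a configured offset to Z.
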(event EventBeginPlay)

(event Collision|EventActorBeginOverlap
  (params (OtherActor :type Actor%20Object%20Reference)))

(event EventTick
  (params (DeltaSeconds :type Float%20%28single-precision%29)))

(event EventOnApplication(Cascade)
  (params (MyTarget :type Actor%20Object%20Reference) (Parameters :type Gameplay%20Cue%20Parameters%20Structure%20%28by%20ref%29) (ParticleComponents :type Array%20of%20Cascade%20Particle%20System%20Component%20Object%20References) (AudioComponents :type Array%20of%20Audio%20Component%20Object%20References) (BurstCameraShakeInstance :type Camera%20Shake%20Base%20Object%20Reference) (BurstDecalInstance :type Decal%20Actor%20Object%20Reference))
  (|CallParentFunction MyTarget Parameters ParticleComponents AudioComponents BurstCameraShakeInstance BurstDecalInstance)
  ;[1 0.127 0.6774] Play Alien VO
  (bind _asfort_player_pawn (Utilities|Casting|CastToFortPlayerPawn MyTarget))
  (Utilities|IsValid _asfort_player_pawn
    (:Is%20Valid
      (if (HelperFunctions|IsLocalPawn _asfort_player_pawn)
        (Audio|PlaySoundatLocation 0 (Variables|Sound|GetSFX_Skyfire_Prison_Warp_1P) (Math|Vector|MakeVector (.x (Transformation|GetWorldLocation (Variables|/Script/Engine.Character|GetMesh _asfort_player_pawn))) (Variables|Sound|GetSFX_PrisonVO_Y_Mod_Value) (+ (.z (Transformation|GetWorldLocation (Variables|/Script/Engine.Character|GetMesh _asfort_player_pawn))) (Variables|Sound|GetSFX_PrisonVO_Z_Mod_Value))))))))

(fn ConstructionScript)
