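; Looping gameplay cue notify for the end of the Skyfire mothership loot room, handling the exit audio.
; On application it runs the parent logic, casts the target to a Fort player pawn and plays a 2D exit sound.
; It then attaches purge alert and sonic boom sounds to the pawn's mesh, using first-person variants when the pawn is local.
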(event EventOnApplication(Cascade)
  (params (MyTarget :type Actor%20Object%20Reference) (Parameters :type Gameplay%20Cue%20Parameters%20Structure%20%28by%20ref%29) (ParticleComponents :type Array%20of%20Cascade%20Particle%20System%20Component%20Object%20References) (AudioComponents :type Array%20of%20Audio%20Component%20Object%20References) (BurstCameraShakeInstance :type Camera%20Shake%20Base%20Object%20Reference) (BurstDecalInstance :type Decal%20Actor%20Object%20Reference))
  ;[1 0.3125 0.8986] Play Exit Mothership Sound
  (|CallParentFunction MyTarget Parameters ParticleComponents AudioComponents BurstCameraShakeInstance BurstDecalInstance)
  (bind _asfort_player_pawn (Utilities|Casting|CastToFortPlayerPawn MyTarget))
  (Audio|PlaySound2D 0 (Variables|Sound|GetSFX_ExitLootRoom))
  (Audio|SpawnSoundAttached (select (HelperFunctions|IsLocalPawn _asfort_player_pawn) (Variables|Sound|GetSFX_Mothership_PurgeAlert_1P) (Variables|Sound|GetSFX_Mothership_PurgeAlert)) (Variables|/Script/Engine.Character|GetMesh _asfort_player_pawn))
  (Audio|SpawnSoundAttached (select (HelperFunctions|IsLocalPawn _asfort_player_pawn) (Variables|Sound|GetSFX_Mothership_Sonicboom_1P) (Variables|Sound|GetSFX_Mothership_Sonicboom)) (Variables|/Script/Engine.Character|GetMesh _asfort_player_pawn)))

(fn ConstructionScript)
