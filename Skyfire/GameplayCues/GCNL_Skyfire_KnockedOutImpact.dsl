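; Gameplay cue notify for the Skyfire knocked-out impact, animating an impact effect on a static mesh's materials.
; On any cue event, it casts the cue's source object, rather than the target, to a static mesh component and stores it.
; The UpdateMaterials timeline then plays from the start, writing its value each update to the impact scalar parameter on that mesh's materials.
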(event GameplayCueNotify|EventHandleGameplayCue
  (params (MyTarget :type Actor%20Object%20Reference) (EventType :type EGameplayCueEvent%20Enum :default "OnActive") (Parameters :type Gameplay%20Cue%20Parameters%20Structure%20%28by%20ref%29))
  (bind (_normalizedmagnitude _rawmagnitude _effectcontext _matchedtagname _originaltag _aggregatedsourcetags _aggregatedtargettags _location _normal _instigator _effectcauser _sourceobject) (Ability|GameplayCue|BreakGameplayCueParameters Parameters))
  ;[0.9346 0 1] Store the source object and update the Impact parameter through a Timeline to adjust the Impact value.
  (bind _asstatic_mesh_component (Utilities|Casting|CastToStaticMeshComponent _sourceobject))
  (Variables|Default|SetMesh _asstatic_mesh_component)
  (bind (_direction _paramvalue) (Timeline|UpdateMaterials PlayFromStart)
    (:Update
      (Rendering|Material|SetScalarParameterValueonMaterials (Variables|Default|GetMesh) (Variables|Default|GetImpactParam) _paramvalue))))

(fn ConstructionScript)
