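; Gameplay cue notify for Skyfire, named for the prison-to-loot-room transition, that applies the Athena UI sound mix modifier while the cue is active.
; It casts the cue target to a FortPlayerPawn and acts only when that pawn is locally controlled.
; OnActive pushes the modifier and Removed pops it, so the audio change affects only that local player.
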(event GameplayCueNotify|EventHandleGameplayCue
  (params (MyTarget :type Actor%20Object%20Reference) (EventType :type EGameplayCueEvent%20Enum :default "OnActive") (Parameters :type Gameplay%20Cue%20Parameters%20Structure%20%28by%20ref%29))
  (bind _athena_ui_mix_modifier (Variables|Default|GetAthena%20UI%20Mix%20Modifier))
  ;[1 0.08977 0.8189] Push/Pop UI Mix Modifier
  (bind _asfort_player_pawn (Utilities|Casting|CastToFortPlayerPawn MyTarget))
  (if (Pawn|IsLocallyControlled _asfort_player_pawn)
    (switch Utilities|FlowControl|Switch|SwitchonEGameplayCueEvent EventType
      (:OnActive
        (Audio|PushSoundMixModifier 0 _athena_ui_mix_modifier))
      (:Removed
        (Audio|PopSoundMixModifier 0 _athena_ui_mix_modifier)))))

(fn ConstructionScript)
