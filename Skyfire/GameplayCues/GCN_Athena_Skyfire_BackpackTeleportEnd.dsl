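; Gameplay cue notify for the end of the Skyfire backpack teleport, providing audio for the teleported actor.
; On execute, it validates the target and checks whether it is the local pawn.
; For a valid non-local pawn only, it plays the player respawn sound at the target's root component location, then returns the parent implementation's result.
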(fn OnExecute
  (params (MyTarget :type Actor%20Object%20Reference) (Parameters :type Gameplay%20Cue%20Parameters%20Structure%20%28by%20ref%29))
  (Utilities|IsValid MyTarget
    (:Is%20Valid
      (if (not (HelperFunctions|IsLocalPawn MyTarget))
        (Audio|PlaySoundatLocation 0 (Variables|Sound|GetSFX_Player_Respawn) (Transformation|GetWorldLocation (Variables|/Script/Engine.Actor|GetRootComponent MyTarget)))
        (return (|CallParentFunction MyTarget Parameters))))))
