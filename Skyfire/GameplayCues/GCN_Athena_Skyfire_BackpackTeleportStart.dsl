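; Gameplay cue notify for the Skyfire backpack teleport start that shuts off pre-teleport effects on the instigating player pawn.
; OnExecute calls DeactivateFXOnInstigator when bDeactivateNiagaraFX is set, then runs the parent implementation.
; That function deactivates Niagara components under the Backpack mesh, plus those under the pawn mesh whose asset matches PawnRingPreTeleportEffect.
; An invalid pawn, backpack or mesh prints a debug string.
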(fn DeactivateFXOnInstigator
  (params (InInstigatorActor :type Actor%20Object%20Reference))
  ;[0 1 0.05893] DeactivateFXOnInstigator - Remove all Niagara backpack effects and a specific pawn effect.
  (bind _asfort_player_pawn (Utilities|Casting|CastToFortPlayerPawn InInstigatorActor))
  (Utilities|IsValid _asfort_player_pawn
    (:Is%20Valid
      (Utilities|FlowControl|Sequence
        (:then_0
          ;;[0.2066 0.05516 0.6667] Deactivate backpack effects.
          (Utilities|IsValid (Customization|GetSkeletalMeshforPartType _asfort_player_pawn "Backpack" :!class /Script/FortniteGame.FortPlayerPawn)
            (:Is%20Valid
              (for _array_element (Components|GetChildrenComponents (Customization|GetSkeletalMeshforPartType _asfort_player_pawn "Backpack" :!class /Script/FortniteGame.FortPlayerPawn) true)
                (bind _asniagara_particle_system_component (Utilities|Casting|CastToNiagaraComponent _array_element))
                (Components|Activation|Deactivate _asniagara_particle_system_component)))
            (:Is%20Not%20Valid
              (Development|PrintString 0 "GCN_Athena_Skyfire_BackpackTeleportStart::DeactivateFXOnInstigator  Backpack is NOT valid." false true "(R=1.000000,G=0.000000,B=0.007289,A=1.000000)" 1e+01))))
        (:then_1
          ;;[0.2433 0.07432 1] Deactivate pawn's ring pre-teleport effect.
          (bind _mesh (Variables|/Script/Engine.Character|GetMesh _asfort_player_pawn)
            (:then
              (bind _array_element_2 (Utilities|Array|ForEachLoopwithBreak (Components|GetChildrenComponents _mesh true))
                (:LoopBody
                  (bind _asniagara_particle_system_component_2 (Utilities|Casting|CastToNiagaraComponent _array_element_2))
                  (if (Utilities|Equal(Object) (Effects|Components|ParticleSystem|GetFXSystemAsset _asniagara_particle_system_component_2) (Variables|Default|GetPawnRingPreTeleportEffect))
                    (Components|Activation|Deactivate _asniagara_particle_system_component_2)))))
            (:else
              (Development|PrintString 0 "GCN_Athena_Skyfire_BackpackTeleportStart::DeactivateFXOnInstigator Pawn's mesh is invalid." false true "(R=1.000000,G=0.000000,B=0.023061,A=1.000000)"))))))
    (:Is%20Not%20Valid
      (Development|PrintString 0 "GCN_Athena_Skyfire_BackpackTeleportStart::DeactivateFXOnInstigator FortPlayerPawnInvalid" false true "(R=1.000000,G=0.000000,B=0.007289,A=1.000000)" 1e+01))))

(fn OnExecute
  (params (MyTarget :type Actor%20Object%20Reference) (Parameters :type Gameplay%20Cue%20Parameters%20Structure%20%28by%20ref%29))
  (bind (_normalizedmagnitude _rawmagnitude _effectcontext _matchedtagname _originaltag _aggregatedsourcetags _aggregatedtargettags _location _normal _instigator) (Ability|GameplayCue|BreakGameplayCueParameters Parameters))
  ;[0.0908 1 0.002553] OnExecute - If DeactivateNigaraFX is true, deactivate the instigator's Niagara effects.
  (Utilities|FlowControl|Sequence
    (:then_0
      (if (Variables|Default|GetbDeactivateNiagaraFX)
        (CallFunction|DeactivateFXOnInstigator _instigator)))
    (:then_1
      (return (|CallParentFunction MyTarget Parameters)))))
