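; Gameplay cue notify for the Skyfire jump in Athena, adding particles, animation and sound to a jumping player.
; OnBurst attaches the first valid Niagara component to a socket on the instigator's Backpack mesh.
; OnExecute triggers a jump anim input event on the target pawn, then spawns a single or double jump sound on its mesh.
; Clamped jump count selects the sound, with a first-person single-jump variant for the local pawn.
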(fn OnBurst(Niagara)
  (params (MyTarget :type Actor%20Object%20Reference) (Parameters :type Gameplay%20Cue%20Parameters%20Structure%20%28by%20ref%29) (ParticleComponents :type Array%20of%20Niagara%20Particle%20System%20Component%20Object%20References) (AudioComponents :type Array%20of%20Audio%20Component%20Object%20References) (BurstCameraShakeInstance :type Camera%20Shake%20Base%20Object%20Reference) (BurstDecalInstance :type Decal%20Actor%20Object%20Reference))
  (bind (_normalizedmagnitude _rawmagnitude _effectcontext _matchedtagname _originaltag _aggregatedsourcetags _aggregatedtargettags _location _normal _instigator) (Ability|GameplayCue|BreakGameplayCueParameters Parameters))
  ;[1 0.865 0.4948] OnBurst - Attach particles to backpack.
  (Utilities|IsValid _instigator
    (:Is%20Valid
      (bind _asplayer_pawn_athena (Utilities|Casting|CastToPlayerPawn_Athena _instigator)
        (:then
          (Utilities|IsValid (Customization|GetSkeletalMeshforPartType _asplayer_pawn_athena "Backpack" :!class /Script/FortniteGame.FortPlayerPawn)
            (:Is%20Valid
              (if (Transformation|DoesSocketExist (Customization|GetSkeletalMeshforPartType _asplayer_pawn_athena "Backpack" :!class /Script/FortniteGame.FortPlayerPawn) (Variables|Default|GetAttachSocketName))
                (for _array_element ParticleComponents
                  (Utilities|IsValid _array_element
                    (:Is%20Valid
                      (Transformation|AttachComponentToComponent _array_element (Customization|GetSkeletalMeshforPartType _asplayer_pawn_athena "Backpack" :!class /Script/FortniteGame.FortPlayerPawn) (Variables|Default|GetAttachSocketName) "KeepRelative" "KeepRelative" "KeepRelative" true)
                      (return))
                    (:Is%20Not%20Valid
                      (Development|PrintString 0 "GCN_Athena_Skyfire_Jump::OnBurst Warning: ParticleCompnent null." false))))
                (else
                  (Development|PrintString 0 "GCN_Athena_Skyfire_Jump::OnBurst Warning: AttachSocketName does not exist." false))))
            (:Is%20Not%20Valid
              (Development|PrintString 0 "GCN_Athena_Skyfire_Jump::OnBurst Warning: Backpack null." false))))
        (:CastFailed
          (Development|PrintString 0 "GCN_Athena_Skyfire_Jump::OnBurst Warning:Cast to PlayerPawn_Athena failed." false))))
    (:Is%20Not%20Valid
      (Development|PrintString 0 "GCN_Athena_Skyfire_Jump::OnBurst Warning: Instigator null." false))))

(fn OnExecute
  (params (MyTarget :type Actor%20Object%20Reference) (Parameters :type Gameplay%20Cue%20Parameters%20Structure%20%28by%20ref%29))
  (bind _sfx_jump_single (Variables|Sound|GetSFX_Jump_Single))
  (bind _sfx_jump_single_1p (Variables|Sound|GetSFX_Jump_Single_1P))
  (bind _sfx_jump_double (Variables|Sound|GetSFX_Jump_Double))
  (bind _min_jumpcount (Variables|Sound|GetMin_JumpCount))
  (bind _max_jumpcount (Variables|Sound|GetMax_JumpCount))
  (Utilities|IsValid MyTarget
    (:Is%20Valid
      (bind _asplayer_pawn_athena (Utilities|Casting|CastToPlayerPawn_Athena MyTarget)
        (:then
          (Utilities|FlowControl|Sequence
            (:then_0
              ;[0.2001 1 0.9569] Trigger Jump anim event
              (bind _returnvalue (Class|FortAnimInputEvent|CreateAnimInputEvent (Variables|Default|GetJumpAnimEventName)))
              (Animation|TriggerAnimInputEvent _asplayer_pawn_athena _returnvalue))
            (:then_1
              ;[1 0.2906 0.7903] Play Jump Sound
              (if (HelperFunctions|IsLocalPawn _asplayer_pawn_athena)
                (Utilities|IsValid (Utilities|Select :Index (Math|Integer|Clamp(Integer) (Variables|/Script/Engine.Character|GetJumpCurrentCount _asplayer_pawn_athena) _min_jumpcount _max_jumpcount) :Option%200 _sfx_jump_single_1p :Option%201 _sfx_jump_double)
                  (:Is%20Valid
                    (Audio|SpawnSoundAttached (Utilities|Select :Index (Math|Integer|Clamp(Integer) (Variables|/Script/Engine.Character|GetJumpCurrentCount _asplayer_pawn_athena) _min_jumpcount _max_jumpcount) :Option%200 _sfx_jump_single_1p :Option%201 _sfx_jump_double) (Variables|/Script/Engine.Character|GetMesh _asplayer_pawn_athena))
                    (return (|CallParentFunction MyTarget Parameters)))
                  (:Is%20Not%20Valid
                    _))
                (else
                  (Utilities|IsValid (Utilities|Select :Index (Math|Integer|Clamp(Integer) (Variables|/Script/Engine.Character|GetJumpCurrentCount _asplayer_pawn_athena) _min_jumpcount _max_jumpcount) :Option%200 _sfx_jump_single :Option%201 _sfx_jump_double)
                    (:Is%20Valid
                      (Audio|SpawnSoundAttached (Utilities|Select :Index (Math|Integer|Clamp(Integer) (Variables|/Script/Engine.Character|GetJumpCurrentCount _asplayer_pawn_athena) _min_jumpcount _max_jumpcount) :Option%200 _sfx_jump_single :Option%201 _sfx_jump_double) (Variables|/Script/Engine.Character|GetMesh _asplayer_pawn_athena)))
                    (:Is%20Not%20Valid
                      _))))
              (Development|PrintString 0 "Sound is not valid  in GCN Athena Skyfire Jump" true true "(R=0.000000,G=0.660000,B=1.000000,A=1.000000)" 22222.0))))
        (:CastFailed
          (Development|PrintString 0 "pawn player is not valid in GCN Athena Skyfire Jump" true true "(R=0.000000,G=0.660000,B=1.000000,A=1.000000)" 22222.0))))
    (:Is%20Not%20Valid
      (Development|PrintString 0 "My target not valid in GCN Athena Skyfire Jump" true true "(R=0.000000,G=0.660000,B=1.000000,A=1.000000)" 22222.0))))
