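; Gameplay cue notify for the Skyfire jump from ground, updating the player's backpack to show remaining jumps or boosts.
; On execute, it casts the cue instigator to PlayerPawn_Athena and fetches its Backpack skeletal mesh.
; It writes the RemainingJumps variable to the jump-count scalar parameter on that mesh's materials, or prints a debug message if the mesh is invalid.
; A Sequence then returns the parent OnExecute result.
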(fn OnExecute
  (params (MyTarget :type Actor%20Object%20Reference) (Parameters :type Gameplay%20Cue%20Parameters%20Structure%20%28by%20ref%29))
  (bind (_normalizedmagnitude _rawmagnitude _effectcontext _matchedtagname _originaltag _aggregatedsourcetags _aggregatedtargettags _location _normal _instigator) (Ability|GameplayCue|BreakGameplayCueParameters Parameters))
  ;[0.0908 1 0.002553] OnExecute - Set the jump count material parameter to the number of remaining jumps/backpack boosts.
  (Utilities|FlowControl|Sequence
    (:then_0
      (bind _asplayer_pawn_athena (Utilities|Casting|CastToPlayerPawn_Athena _instigator))
      (Utilities|IsValid (Customization|GetSkeletalMeshforPartType _asplayer_pawn_athena "Backpack" :!class /Script/FortniteGame.FortPlayerPawn)
        (:Is%20Valid
          (Rendering|Material|SetScalarParameterValueonMaterials (Customization|GetSkeletalMeshforPartType _asplayer_pawn_athena "Backpack" :!class /Script/FortniteGame.FortPlayerPawn) (Variables|Default|GetJumpCountMaterialParameterName) (Variables|Default|GetRemainingJumps)))
        (:Is%20Not%20Valid
          (Development|PrintString 0 "GCN_Athena_Skyfire_JumpFromGround::OnExecute Backpack is NOT valid." false true "(R=1.000000,G=0.000000,B=0.030066,A=1.000000)" 1e+01))))
    (:then_1
      (return (|CallParentFunction MyTarget Parameters)))))
