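; Gameplay cue notify actor for Skyfire that shows a low-gravity Niagara trail on a player pawn, colored by earned loot rarity.
; OnBecomeRelevant attaches the trail to the pawn mesh and sets its max draw distance; OnCeaseRelevant deactivates it.
; HandleGameplayCue binds to the SkyfirePlayerPawnComponent's OnEarnedLootRarityChanged without filtering by event type.
; That callback pushes a color from the new rarity's palette into a Niagara parameter; no unbind is present.
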(event EventBeginPlay)

(event Collision|EventActorBeginOverlap
  (params (OtherActor :type Actor%20Object%20Reference)))

(event GameplayCueNotify|EventHandleGameplayCue
  (params (MyTarget :type Actor%20Object%20Reference) (EventType :type EGameplayCueEvent%20Enum :default "OnActive") (Parameters :type Gameplay%20Cue%20Parameters%20Structure%20%28by%20ref%29))
  (bind _returnvalue (Actor|GetComponentbyClass MyTarget "/Script/SkyfireRuntime.SkyfirePlayerPawnComponent"))
  ;[0.7268 0 1] Update Low Grav Trail FX Colors
  (Utilities|IsValid MyTarget
    (:Is%20Valid
      (Utilities|IsValid _returnvalue
        (:Is%20Valid
          (EventDispatchers|BindEventtoOnEarnedLootRarityChanged _returnvalue (EventDispatchers|CreateEvent :!event Custom|ChangeFXColors)))
        (:Is%20Not%20Valid
          (Development|PrintString 0 "invalid component class"))))
    (:Is%20Not%20Valid
      (Development|PrintString 0 "invalid player pawn"))) ; If the cached component isn't valid, find it
)

(event Custom|ChangeFXColors
  (params (NewRarity :type EFortRarity%20Enum :default "Common"))
  ;[0.7268 0 1] Update Low Grav Trail FX Colors
  (bind _returnvalue (Fortnite|GetRarityColorPalette NewRarity))
  (Niagara|SetNiagaraVariable(LinearColor) (Variables|Default|GetAttachedLowGravTrailFX) (Variables|Default|GetNiagaraParameterName) (Utilities|Struct|BreakFortColorPalette _returnvalue)))

(fn OnCeaseRelevant
  (params (MyTarget :type Actor%20Object%20Reference) (Parameters :type Gameplay%20Cue%20Parameters%20Structure%20%28by%20ref%29))
  ;[0.8634 0 1] Disable Low Grav FX
  (Actor|Tick|SetActorTickEnabled false)
  (bind _attachedlowgravtrailfx (Variables|Default|GetAttachedLowGravTrailFX)
    (:then
      (Components|Activation|Deactivate _attachedlowgravtrailfx)
      (return true))
    (:else
      (return false))))

(fn ConstructionScript)

(fn OnBecomeRelevant
  (params (MyTarget :type Actor%20Object%20Reference) (Parameters :type Gameplay%20Cue%20Parameters%20Structure%20%28by%20ref%29))
  ;[0.7792 0.03417 1] Play Low Grav Trail FX
  (bind _asfort_player_pawn_athena (Utilities|Casting|CastToFortPlayerPawnAthena MyTarget))
  (Actor|Tick|SetActorTickEnabled true)
  (Utilities|IsValid (Variables|/Script/Engine.Character|GetMesh _asfort_player_pawn_athena)
    (:Is%20Valid
      (bind _returnvalue (Niagara|SpawnSystemAttached (Variables|Default|GetLowGravFX) (Variables|/Script/Engine.Character|GetMesh _asfort_player_pawn_athena)))
      (bind _output_get (Variables|Default|SetAttachedLowGravTrailFX _returnvalue))
      (LOD|SetMaxDrawDistance _output_get (Variables|Default|GetMaxDrawDistance))
      (return true))
    (:Is%20Not%20Valid
      (return false))))
